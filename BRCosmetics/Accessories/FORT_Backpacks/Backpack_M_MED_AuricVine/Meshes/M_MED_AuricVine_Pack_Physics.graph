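# Physics asset for the AuricVine backpack cosmetic mesh, giving its dangling leaf bones collision bodies and joints for secondary motion.
# Thirteen dyn_leaf bones each carry a small capsule offset along the bone, while the pack bone has a single small sphere.
# Every leaf is constrained directly to the pack body in a star layout rather than chains, with individually tuned swing and twist limits bounding how far it moves.
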PhysicsAsset /BRCosmetics/Accessories/FORT_Backpacks/Backpack_M_MED_AuricVine/Meshes/M_MED_AuricVine_Pack_Physics
bodies {
 dyn_leaf_a_01 [{shapeType=Capsule, length=1.5, radius=1, shapeName=dyn_leaf_a_01_capsule, center=<2 0 0>, rotation=<rot 90 0 0>}]
 dyn_leaf_ba_01 [{shapeType=Capsule, length=1, radius=1, shapeName=dyn_leaf_ba_01_capsule, center=<2.5 0 0>, rotation=<rot 90 0 0>}]
 dyn_leaf_bb_01 [{shapeType=Capsule, length=1, radius=1, shapeName=dyn_leaf_bb_01_capsule, center=<2.5 0 0>, rotation=<rot 90 0 0>}]
 dyn_leaf_c_01 [{shapeType=Capsule, length=1.5, radius=1, shapeName=dyn_leaf_c_01_capsule, center=<1 0 0>, rotation=<rot 90 0 0>}]
 dyn_leaf_da_01 [{shapeType=Capsule, length=1, radius=1, shapeName=dyn_leaf_da_01_capsule, center=<2.5 0 0>, rotation=<rot 90 0 0>}]
 dyn_leaf_db_01 [{shapeType=Capsule, length=1, radius=1, shapeName=dyn_leaf_db_01_capsule, center=<2.5 0 0>, rotation=<rot 90 0 0>}]
 dyn_leaf_dc_01 [{shapeType=Capsule, length=1, radius=1, shapeName=dyn_leaf_dc_01_capsule, center=<2.5 0 0>, rotation=<rot -90 0 0>}]
 dyn_leaf_ea_01 [{shapeType=Capsule, length=1, radius=1, shapeName=dyn_leaf_ea_01_capsule, center=<2.5 0 0>, rotation=<rot 90 0 0>}]
 dyn_leaf_eb_01 [{shapeType=Capsule, length=1, radius=1, shapeName=dyn_leaf_eb_01_capsule, center=<2.5 0 0>, rotation=<rot 90 0 0>}]
 dyn_leaf_f_01 [{shapeType=Capsule, length=1.5, radius=1, shapeName=dyn_leaf_f_01_capsule, center=<1 0 0>, rotation=<rot 90 0 0>}]
 dyn_leaf_g_01 [{shapeType=Capsule, length=1.5, radius=1, shapeName=dyn_leaf_g_01_capsule, center=<1 0 0>, rotation=<rot 90 0 0>}]
 dyn_leaf_ha_01 [{shapeType=Capsule, length=1, radius=1, shapeName=dyn_leaf_ha_01_capsule, center=<2.5 0 0>, rotation=<rot 90 0 0>}]
 dyn_leaf_hb_01 [{shapeType=Capsule, length=1, radius=1, shapeName=dyn_leaf_hb_01_capsule, center=<2.5 0 0>, rotation=<rot 90 0 0>}]
 pack [{shapeType=Sphere, radius=1, shapeName=pack_sphere}]
}
constraints [
 {bone1Name=dyn_leaf_a_01, bone2Name=pack, swing1LimitDegrees=45, swing2LimitDegrees=60, twistLimitDegrees=36.6}
 {bone1Name=dyn_leaf_c_01, bone2Name=pack, swing1LimitDegrees=67.36, swing2LimitDegrees=53.64, twistLimitDegrees=15}
 {bone1Name=dyn_leaf_f_01, bone2Name=pack, swing1LimitDegrees=55, swing2LimitDegrees=60, twistLimitDegrees=22.2}
 {bone1Name=dyn_leaf_g_01, bone2Name=pack, swing1LimitDegrees=35, swing2LimitDegrees=45, twistLimitDegrees=23.64}
 {bone1Name=dyn_leaf_ba_01, bone2Name=pack, swing1LimitDegrees=50.08, swing2LimitDegrees=60, twistLimitDegrees=26.52}
 {bone1Name=dyn_leaf_bb_01, bone2Name=pack, swing1LimitDegrees=30.08, swing2LimitDegrees=45, twistLimitDegrees=26.52}
 {bone1Name=dyn_leaf_da_01, bone2Name=pack, swing1LimitDegrees=40, swing2LimitDegrees=40, twistLimitDegrees=15}
 {bone1Name=dyn_leaf_db_01, bone2Name=pack, swing1LimitDegrees=35, swing2LimitDegrees=35, twistLimitDegrees=20}
 {bone1Name=dyn_leaf_ea_01, bone2Name=pack, swing1LimitDegrees=35, swing2LimitDegrees=35, twistLimitDegrees=15}
 {bone1Name=dyn_leaf_eb_01, bone2Name=pack, swing1LimitDegrees=45, swing2LimitDegrees=45, twistLimitDegrees=15}
 {bone1Name=dyn_leaf_ha_01, bone2Name=pack, swing1LimitDegrees=35, swing2LimitDegrees=35, twistLimitDegrees=10}
 {bone1Name=dyn_leaf_hb_01, bone2Name=pack, swing1LimitDegrees=40, swing2LimitDegrees=42.28, twistLimitDegrees=26.52}
 {bone1Name=dyn_leaf_dc_01, bone2Name=pack, swing1LimitDegrees=40, swing2LimitDegrees=40, twistLimitDegrees=15}
]
