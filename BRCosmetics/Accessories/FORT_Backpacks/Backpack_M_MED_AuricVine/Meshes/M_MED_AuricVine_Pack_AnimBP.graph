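# Post-process style AnimBlueprint for the AuricVine backpack cosmetic, adding secondary motion to its leaf bones on top of the input pose.
# Selects RigidBody physics, else chained AnimDynamics with spring and spherical limits, else planar-limited Trail controllers, using two enable flags.
# A small additive root-bone offset follows the selection.
# The update event refreshes gravity override for both physics paths through a deprecated state function.
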AnimBlueprint /BRCosmetics/Accessories/FORT_Backpacks/Backpack_M_MED_AuricVine/Meshes/M_MED_AuricVine_Pack_AnimBP
target_skeleton=&/M_MED_AuricVine_Pack_Skeleton
graphs [
 AnimGraph {nodes [
   n1 = Blend Poses by bool(
    BlendPose_0=Use cached pose 'Cache_AnimDynamics'()
    BlendPose_1=Use cached pose 'Cache_Trails'()
    bActiveValue=Get bAnimDynamicsEnabled() -> (bAnimDynamicsEnabled:bool)
   ) {blendType=Linear}
   n2 = Use cached pose 'Cache_RBAN'()
   n3 = Get bRigidBodyEnabled()
   n4 = "Transform (Modify) Bone - Bone: root"(
    ComponentPose=Local To Component(LocalPose=Blend Poses by bool(BlendPose_0=@n2, BlendPose_1=@n1, bActiveValue=@n3) {blendType=Linear})
    Translation="0,0.5,-3"
    Rotation="0,0,0"
   ) {
    translationMode=BMM_Additive
    translationSpace=BCS_BoneSpace
    alphaBoolBlend {blendOption=Linear}
    boneToModify {boneName=root}
    translation {y=0.5, z=-3}
   }
   Output Pose(Result=Component To Local(ComponentPose=@n4)) {layerGroup=None}
   Save cached pose 'Cache_BaseInput'(Pose=Input Pose()) {cacheName=Cache_BaseInput}
   n5 = Get Gravity Override()
   n6 = Anim Dynamics - Bone: dyn_leaf_a_01!(ComponentPose=Local To Component(LocalPose=Use cached pose 'Cache_BaseInput'()), Alpha=1, GravityOverride:Vector=@n5) -> (Pose) {
    angularSpringConstant=0.999
    bAngularSpring=true
    bUseGravityOverride=true
    bUseSphericalLimits=true
    simulationSpace=BoneRelative
    alphaBoolBlend {blendOption=Linear}
    boundBone {boneName=dyn_leaf_a_01}
    componentAppliedLinearAccClamp=<3000 3000 3000>
    componentLinearAccScale=<0.5 0.5 1>
    componentLinearVelScale=<0.5 0.5 1>
    physicsBodyDefinitions [{
      boundBone {boneName=dyn_leaf_a_01}
      boxExtents=<5 10 10>
      constraintSetup {angularLimitsMax=<30 30 30>, angularLimitsMin=<-30 -30 -30>, angularTarget=<0 0 0>}
      localJointOffset=<1 0 0>
     }]
    relativeSpaceBone {boneName=pack}
    sphericalLimits [{limitRadius=0, drivingBone {boneName=None}, sphereLocalOffset=<0 0 0>}]
   }
   n7 = Anim Dynamics - Bone: dyn_leaf_ba_01!(ComponentPose=@n6, Alpha=1, GravityOverride:Vector=@n5) -> (Pose) {
    angularSpringConstant=0.999
    bAngularSpring=true
    bUseGravityOverride=true
    bUseSphericalLimits=true
    simulationSpace=BoneRelative
    alphaBoolBlend {blendOption=Linear}
    boundBone {boneName=dyn_leaf_ba_01}
    componentAppliedLinearAccClamp=<3000 3000 3000>
    componentLinearAccScale=<0.5 0.5 1>
    componentLinearVelScale=<0.5 0.5 1>
    physicsBodyDefinitions [{
      boundBone {boneName=dyn_leaf_ba_01}
      boxExtents=<10 15 15>
      constraintSetup {angularLimitsMax=<30 30 30>, angularLimitsMin=<-30 -30 -30>, angularTarget=<0 0 0>}
      localJointOffset=<0.75 0 0>
     }]
    relativeSpaceBone {boneName=pack}
    sphericalLimits [{limitRadius=4, drivingBone {boneName=dyn_leaf_bb_02}, sphereLocalOffset=<0 0 0>}]
   }
   n8 = Anim Dynamics - Bone: dyn_leaf_bb_01!(ComponentPose=@n7, Alpha=1, GravityOverride:Vector=@n5) -> (Pose) {
    angularSpringConstant=0.999
    bAngularSpring=true
    bUseGravityOverride=true
    bUseSphericalLimits=true
    simulationSpace=BoneRelative
    alphaBoolBlend {blendOption=Linear}
    boundBone {boneName=dyn_leaf_bb_01}
    componentAppliedLinearAccClamp=<3000 3000 3000>
    componentLinearAccScale=<0.5 0.5 1>
    componentLinearVelScale=<0.5 0.5 1>
    physicsBodyDefinitions [{
      boundBone {boneName=dyn_leaf_bb_01}
      boxExtents=<15 20 20>
      constraintSetup {angularLimitsMax=<15 30 15>, angularLimitsMin=<-15 -60 -15>, angularTarget=<0 0 0>}
      localJointOffset=<1 0 0>
     }]
    relativeSpaceBone {boneName=pack}
    sphericalLimits [{limitRadius=1, drivingBone {boneName=dyn_leaf_ba_02}, sphereLocalOffset=<0 0 0>}]
   }
   n9 = Anim Dynamics - Bone: dyn_leaf_c_01!(ComponentPose=@n8, Alpha=1, GravityOverride:Vector=@n5) -> (Pose) {
    angularSpringConstant=0.999
    bAngularSpring=true
    bUseGravityOverride=true
    bUseSphericalLimits=true
    simulationSpace=BoneRelative
    alphaBoolBlend {blendOption=Linear}
    boundBone {boneName=dyn_leaf_c_01}
    componentAppliedLinearAccClamp=<3000 3000 3000>
    componentLinearAccScale=<0.5 0.5 1>
    componentLinearVelScale=<0.5 0.5 1>
    physicsBodyDefinitions [{
      boundBone {boneName=dyn_leaf_c_01}
      boxExtents=<5 10 10>
      constraintSetup {angularLimitsMax=<30 30 30>, angularLimitsMin=<-30 -30 -30>, angularTarget=<0 0 0>}
      localJointOffset=<1 0 0>
     }]
    relativeSpaceBone {boneName=pack}
    sphericalLimits [{limitRadius=0, drivingBone {boneName=None}, sphereLocalOffset=<0 0 0>}]
   }
   n10 = Anim Dynamics - Bone: dyn_leaf_da_01!(ComponentPose=@n9, Alpha=1, GravityOverride:Vector=@n5) -> (Pose) {
    angularSpringConstant=0.999
    bAngularSpring=true
    bUseGravityOverride=true
    bUseSphericalLimits=true
    simulationSpace=BoneRelative
    alphaBoolBlend {blendOption=Linear}
    boundBone {boneName=dyn_leaf_da_01}
    componentAppliedLinearAccClamp=<3000 3000 3000>
    componentLinearAccScale=<0.5 0.5 1>
    componentLinearVelScale=<0.5 0.5 1>
    physicsBodyDefinitions [{
      boundBone {boneName=dyn_leaf_da_01}
      boxExtents=<15 20 20>
      constraintSetup {angularLimitsMax=<30 30 30>, angularLimitsMin=<-30 -30 -30>, angularTarget=<0 0 0>}
      localJointOffset=<1 0 0>
     }]
    relativeSpaceBone {boneName=pack}
    sphericalLimits [
     {limitRadius=1.5, drivingBone {boneName=dyn_leaf_dc_01}, sphereLocalOffset=<0 0 0.0069>}
     {limitRadius=1.5, drivingBone {boneName=dyn_leaf_db_02}, sphereLocalOffset=<0 0 0>}
    ]
   }
   n11 = Anim Dynamics - Bone: dyn_leaf_db_01!(ComponentPose=@n10, Alpha=0.8, GravityOverride:Vector=@n5) -> (Pose) {
    angularSpringConstant=0.999
    bAngularSpring=true
    bUseGravityOverride=true
    bUseSphericalLimits=true
    simulationSpace=BoneRelative
    alphaBoolBlend {blendOption=Linear}
    boundBone {boneName=dyn_leaf_db_01}
    componentAppliedLinearAccClamp=<3000 3000 3000>
    componentLinearAccScale=<0.5 0.5 1>
    componentLinearVelScale=<0.5 0.5 1>
    physicsBodyDefinitions [{
      boundBone {boneName=dyn_leaf_db_01}
      boxExtents=<20 25 25>
      constraintSetup {angularLimitsMax=<30 0 30>, angularLimitsMin=<-30 -30 -30>, angularTarget=<0 0 0>}
      localJointOffset=<1.5 0 0>
     }]
    relativeSpaceBone {boneName=pack}
    sphericalLimits [{limitRadius=2, drivingBone {boneName=dyn_leaf_dc_01}, sphereLocalOffset=<0 0 0>}]
   }
   n12 = Anim Dynamics - Bone: dyn_leaf_dc_01!(ComponentPose=@n11, Alpha=0.8, GravityOverride:Vector=@n5) -> (Pose) {
    angularSpringConstant=0.999
    bAngularSpring=true
    bUseGravityOverride=true
    bUseSphericalLimits=true
    simulationSpace=BoneRelative
    alphaBoolBlend {blendOption=Linear}
    boundBone {boneName=dyn_leaf_dc_01}
    componentAppliedLinearAccClamp=<3000 3000 3000>
    componentLinearAccScale=<0.5 0.5 1>
    componentLinearVelScale=<0.5 0.5 1>
    physicsBodyDefinitions [{
      boundBone {boneName=dyn_leaf_dc_01}
      boxExtents=<20 25 25>
      constraintSetup {angularLimitsMax=<30 30 30>, angularLimitsMin=<-30 -30 -30>, angularTarget=<0 0 0>}
      localJointOffset=<1.5 0 0>
     }]
    relativeSpaceBone {boneName=pack}
    sphericalLimits [
     {limitRadius=1.2, drivingBone {boneName=dyn_leaf_db_02}, sphereLocalOffset=<-1.8924 0 0>}
     {limitRadius=2, drivingBone {boneName=dyn_leaf_da_01}, sphereLocalOffset=<0 0 0>}
    ]
   }
   n13 = Anim Dynamics - Bone: dyn_leaf_ea_01!(ComponentPose=@n12, Alpha=1, GravityOverride:Vector=@n5) -> (Pose) {
    angularSpringConstant=0.999
    bAngularSpring=true
    bUseGravityOverride=true
    bUseSphericalLimits=true
    simulationSpace=BoneRelative
    alphaBoolBlend {blendOption=Linear}
    boundBone {boneName=dyn_leaf_ea_01}
    componentAppliedLinearAccClamp=<3000 3000 3000>
    componentLinearAccScale=<0.5 0.5 1>
    componentLinearVelScale=<0.5 0.5 1>
    physicsBodyDefinitions [{
      boundBone {boneName=dyn_leaf_ea_01}
      boxExtents=<10 15 15>
      constraintSetup {angularLimitsMax=<30 30 30>, angularLimitsMin=<-30 -30 -30>, angularTarget=<0 0 0>}
      localJointOffset=<1 0 0>
     }]
    relativeSpaceBone {boneName=pack}
    sphericalLimits [{limitRadius=2, drivingBone {boneName=dyn_leaf_eb_02}, sphereLocalOffset=<0 0 0>}]
   }
   n14 = Anim Dynamics - Bone: dyn_leaf_eb_01!(ComponentPose=@n13, Alpha=1, GravityOverride:Vector=@n5) -> (Pose) {
    angularSpringConstant=0.999
    bAngularSpring=true
    bUseGravityOverride=true
    bUseSphericalLimits=true
    simulationSpace=BoneRelative
    alphaBoolBlend {blendOption=Linear}
    boundBone {boneName=dyn_leaf_eb_01}
    componentAppliedLinearAccClamp=<3000 3000 3000>
    componentLinearAccScale=<0.5 0.5 1>
    componentLinearVelScale=<0.5 0.5 1>
    physicsBodyDefinitions [{
      boundBone {boneName=dyn_leaf_eb_01}
      boxExtents=<15 20 20>
      constraintSetup {angularLimitsMax=<30 30 30>, angularLimitsMin=<-30 -30 -30>, angularTarget=<0 0 0>}
      localJointOffset=<1 0 0>
     }]
    relativeSpaceBone {boneName=pack}
    sphericalLimits [{limitRadius=2, drivingBone {boneName=dyn_leaf_ea_02}, sphereLocalOffset=<0 0 0>}]
   }
   n15 = Anim Dynamics - Bone: dyn_leaf_f_01!(ComponentPose=@n14, Alpha=1, GravityOverride:Vector=@n5) -> (Pose) {
    angularSpringConstant=0.999
    bAngularSpring=true
    bUseGravityOverride=true
    bUseSphericalLimits=true
    simulationSpace=BoneRelative
    alphaBoolBlend {blendOption=Linear}
    boundBone {boneName=dyn_leaf_f_01}
    componentAppliedLinearAccClamp=<3000 3000 3000>
    componentLinearAccScale=<0.5 0.5 1>
    componentLinearVelScale=<0.5 0.5 1>
    physicsBodyDefinitions [{
      boundBone {boneName=dyn_leaf_f_01}
      boxExtents=<5 10 10>
      constraintSetup {angularLimitsMax=<30 30 30>, angularLimitsMin=<-30 -30 -30>, angularTarget=<0 0 0>}
      localJointOffset=<1 0 0>
     }]
    relativeSpaceBone {boneName=pack}
    sphericalLimits [{limitRadius=0, drivingBone {boneName=None}, sphereLocalOffset=<0 0 0>}]
   }
   n16 = Anim Dynamics - Bone: dyn_leaf_g_01!(ComponentPose=@n15, Alpha=1, GravityOverride:Vector=@n5) -> (Pose) {
    angularSpringConstant=0.999
    bAngularSpring=true
    bUseGravityOverride=true
    bUseSphericalLimits=true
    simulationSpace=BoneRelative
    alphaBoolBlend {blendOption=Linear}
    boundBone {boneName=dyn_leaf_g_01}
    componentAppliedLinearAccClamp=<3000 3000 3000>
    componentLinearAccScale=<0.5 0.5 1>
    componentLinearVelScale=<0.5 0.5 1>
    physicsBodyDefinitions [{
      boundBone {boneName=dyn_leaf_g_01}
      boxExtents=<5 7.5 7.5>
      constraintSetup {angularLimitsMax=<30 30 30>, angularLimitsMin=<-30 -30 -30>, angularTarget=<0 0 0>}
      localJointOffset=<0.5 0 0>
     }]
    relativeSpaceBone {boneName=pack}
    sphericalLimits [{limitRadius=0, drivingBone {boneName=None}, sphereLocalOffset=<0 0 0>}]
   }
   n17 = Anim Dynamics - Bone: dyn_leaf_ha_01!(ComponentPose=@n16, Alpha=1, GravityOverride:Vector=@n5) -> (Pose) {
    angularSpringConstant=0.999
    bAngularSpring=true
    bUseGravityOverride=true
    bUseSphericalLimits=true
    simulationSpace=BoneRelative
    alphaBoolBlend {blendOption=Linear}
    boundBone {boneName=dyn_leaf_ha_01}
    componentAppliedLinearAccClamp=<3000 3000 3000>
    componentLinearAccScale=<0.5 0.5 1>
    componentLinearVelScale=<0.5 0.5 1>
    physicsBodyDefinitions [{
      boundBone {boneName=dyn_leaf_ha_01}
      boxExtents=<15 20 20>
      constraintSetup {angularLimitsMax=<15 10 15>, angularLimitsMin=<-15 -30 -15>, angularTarget=<0 0 0>}
      localJointOffset=<0.75 0 0>
     }]
    relativeSpaceBone {boneName=pack}
    sphericalLimits [{limitRadius=1.5, drivingBone {boneName=dyn_leaf_hb_02}, sphereLocalOffset=<0 0 0>}]
   }
   n18 = Anim Dynamics - Bone: dyn_leaf_hb_01!(ComponentPose=@n17, Alpha=1, GravityOverride:Vector=@n5) -> (Pose) {
    angularSpringConstant=0.999
    bAngularSpring=true
    bUseGravityOverride=true
    bUseSphericalLimits=true
    simulationSpace=BoneRelative
    alphaBoolBlend {blendOption=Linear}
    boundBone {boneName=dyn_leaf_hb_01}
    componentAppliedLinearAccClamp=<3000 3000 3000>
    componentLinearAccScale=<0.5 0.5 1>
    componentLinearVelScale=<0.5 0.5 1>
    physicsBodyDefinitions [{
      boundBone {boneName=dyn_leaf_hb_01}
      boxExtents=<10 15 15>
      constraintSetup {angularLimitsMax=<30 30 30>, angularLimitsMin=<-30 -30 -30>, angularTarget=<0 0 0>}
      localJointOffset=<0.5 0 0>
     }]
    relativeSpaceBone {boneName=pack}
    sphericalLimits [{limitRadius=1.2, drivingBone {boneName=dyn_leaf_ha_01}, sphereLocalOffset=<0 0 0>}]
   }
   Save cached pose 'Cache_AnimDynamics'(Pose=Component To Local(ComponentPose=@n18)) {cacheName=Cache_AnimDynamics}
   n19 = RigidBody!(ComponentPose=Local To Component(LocalPose=Use cached pose 'Cache_BaseInput'()), Alpha, OverrideWorldGravity:Vector=Get Gravity Override()) -> (Pose) {
    bClampLinearTranslationLimitToRefPose=true
    bOverrideWorldGravity=true
    overridePhysicsAsset=&/M_MED_AuricVine_Pack_Physics
    simulationSpace=BaseBoneSpace
    alphaBoolBlend {blendOption=Linear}
    baseBoneRef {boneName=root}
    componentAppliedLinearAccClamp=<2500 3000 2000>
    componentLinearAccScale=<1 0.2 0.3>
    componentLinearVelScale=<1 0.2 0.3>
   }
   Save cached pose 'Cache_RBAN'(Pose=Component To Local(ComponentPose=@n19)) {cacheName=Cache_RBAN}
   n20 = Get AlphaTrail() -> (AlphaTrail:float)
   n21 = Trail controller - Bone: dyn_leaf_a_02(ComponentPose=Local To Component(LocalPose=Use cached pose 'Cache_BaseInput'()), Alpha=@n20) {
    bUsePlanarLimit=true
    relaxationSpeedScale=3.5
    planarLimits [
     {drivingBone {boneName=dyn_leaf_a_01}, planeTransform {location=<0 0 0>, rotation=<rot 0 0 0>, scale=<1 1 1>}}
     {drivingBone {boneName=dyn_leaf_a_01}, planeTransform {location=<0 0 0>, rotation=<rot -60 180 -180>, scale=<1 1 1>}}
    ]
    rotationLimits [{limitMax=<180 180 180>}, {limitMax=<180 180 180>}]
    rotationOffsets [<0 0 0>, <0 0 0>]
    trailBone {boneName=dyn_leaf_a_02}
   }
   n22 = Trail controller - Bone: dyn_leaf_ba_02(ComponentPose=@n21, Alpha=@n20) {
    bShowLimit=false
    bUsePlanarLimit=true
    relaxationSpeedScale=4.5
    planarLimits [
     {drivingBone {boneName=dyn_leaf_ba_01}, planeTransform {location=<0 0 1>, rotation=<rot 0 0 0>, scale=<1 1 1>}}
     {drivingBone {boneName=dyn_leaf_ba_01}, planeTransform {location=<2 0 0>, rotation=<rot -60 180 -180>, scale=<1 1 1>}}
    ]
    rotationLimits [{limitMax=<180 180 180>}, {limitMax=<180 180 180>}]
    rotationOffsets [<0 0 0>, <0 0 0>]
    trailBone {boneName=dyn_leaf_ba_02}
   }
   n23 = Trail controller - Bone: dyn_leaf_bb_02(ComponentPose=@n22, Alpha=@n20) {
    bShowLimit=false
    bUsePlanarLimit=true
    relaxationSpeedScale=5.5
    planarLimits [
     {drivingBone {boneName=dyn_leaf_bb_01}, planeTransform {location=<0 0 1>, rotation=<rot 0 0 0>, scale=<1 1 1>}}
     {drivingBone {boneName=dyn_leaf_bb_01}, planeTransform {location=<2 0 0>, rotation=<rot -60 180 -180>, scale=<1 1 1>}}
    ]
    rotationLimits [{limitMax=<180 180 180>}, {limitMax=<180 180 180>}]
    rotationOffsets [<0 0 0>, <0 0 0>]
    trailBone {boneName=dyn_leaf_bb_02}
   }
   n24 = Trail controller - Bone: dyn_leaf_c_02(ComponentPose=@n23, Alpha=@n20) {
    bShowLimit=false
    bUsePlanarLimit=true
    relaxationSpeedScale=5
    planarLimits [
     {drivingBone {boneName=dyn_leaf_c_01}, planeTransform {location=<0 0 0>, rotation=<rot 0 0 0>, scale=<1 1 1>}}
     {drivingBone {boneName=dyn_leaf_c_01}, planeTransform {location=<0 0 0>, rotation=<rot -60 180 -180>, scale=<1 1 1>}}
    ]
    rotationLimits [{limitMax=<180 180 180>}, {limitMax=<180 180 180>}]
    rotationOffsets [<0 0 0>, <0 0 0>]
    trailBone {boneName=dyn_leaf_c_02}
   }
   n25 = Trail controller - Bone: dyn_leaf_da_02(ComponentPose=@n24, Alpha=@n20) {
    bUsePlanarLimit=true
    relaxationSpeedScale=6
    planarLimits [
     {drivingBone {boneName=dyn_leaf_da_01}, planeTransform {location=<0 0 2>, rotation=<rot 0 0 0>, scale=<1 1 1>}}
     {drivingBone {boneName=dyn_leaf_da_01}, planeTransform {location=<4 0 0>, rotation=<rot -60 180 -180>, scale=<1 1 1>}}
    ]
    rotationLimits [{limitMax=<180 180 180>}, {limitMax=<180 180 180>}]
    rotationOffsets [<0 0 0>, <0 0 0>]
    trailBone {boneName=dyn_leaf_da_02}
   }
   n26 = Trail controller - Bone: dyn_leaf_db_02(ComponentPose=@n25, Alpha=@n20) {
    bUsePlanarLimit=true
    relaxationSpeedScale=6.5
    planarLimits [
     {drivingBone {boneName=dyn_leaf_db_01}, planeTransform {location=<0 0 2>, rotation=<rot 0 0 0>, scale=<1 1 1>}}
     {drivingBone {boneName=dyn_leaf_db_01}, planeTransform {location=<4 0 0>, rotation=<rot -60 180 -180>, scale=<1 1 1>}}
    ]
    rotationLimits [{limitMax=<180 180 180>}, {limitMax=<180 180 180>}]
    rotationOffsets [<0 0 0>, <0 0 0>]
    trailBone {boneName=dyn_leaf_db_02}
   }
   n27 = Trail controller - Bone: dyn_leaf_dc_02(ComponentPose=@n26, Alpha=@n20) {
    bUsePlanarLimit=true
    relaxationSpeedScale=6.5
    planarLimits [
     {drivingBone {boneName=dyn_leaf_db_01}, planeTransform {location=<0 0 2>, rotation=<rot 0 0 0>, scale=<1 1 1>}}
     {drivingBone {boneName=dyn_leaf_db_01}, planeTransform {location=<4 0 0>, rotation=<rot -60 180 -180>, scale=<1 1 1>}}
    ]
    rotationLimits [{limitMax=<180 180 180>}, {limitMax=<180 180 180>}]
    rotationOffsets [<0 0 0>, <0 0 0>]
    trailBone {boneName=dyn_leaf_dc_02}
   }
   n28 = Trail controller - Bone: dyn_leaf_ea_02(ComponentPose=@n27, Alpha=@n20) {
    bShowLimit=false
    bUsePlanarLimit=true
    relaxationSpeedScale=5
    planarLimits [
     {drivingBone {boneName=dyn_leaf_ea_01}, planeTransform {location=<0 0 1>, rotation=<rot 0 0 0>, scale=<1 1 1>}}
     {drivingBone {boneName=dyn_leaf_ea_01}, planeTransform {location=<1 0 0>, rotation=<rot -60 180 -180>, scale=<1 1 1>}}
    ]
    rotationLimits [{limitMax=<180 180 180>}, {limitMax=<180 180 180>}]
    rotationOffsets [<0 0 0>, <0 0 0>]
    trailBone {boneName=dyn_leaf_ea_02}
   }
   n29 = Trail controller - Bone: dyn_leaf_eb_02(ComponentPose=@n28, Alpha=@n20) {
    bShowLimit=false
    bUsePlanarLimit=true
    relaxationSpeedScale=5.5
    planarLimits [
     {drivingBone {boneName=dyn_leaf_eb_01}, planeTransform {location=<0 0 1>, rotation=<rot 0 0 0>, scale=<1 1 1>}}
     {drivingBone {boneName=dyn_leaf_eb_01}, planeTransform {location=<1 0 0>, rotation=<rot -60 180 -180>, scale=<1 1 1>}}
    ]
    rotationLimits [{limitMax=<180 180 180>}, {limitMax=<180 180 180>}]
    rotationOffsets [<0 0 0>, <0 0 0>]
    trailBone {boneName=dyn_leaf_eb_02}
   }
   n30 = Trail controller - Bone: dyn_leaf_f_02(ComponentPose=@n29, Alpha=@n20) {
    bShowLimit=false
    bUsePlanarLimit=true
    relaxationSpeedScale=5
    planarLimits [
     {drivingBone {boneName=dyn_leaf_f_01}, planeTransform {location=<0 0 0>, rotation=<rot 0 0 0>, scale=<1 1 1>}}
     {drivingBone {boneName=dyn_leaf_f_01}, planeTransform {location=<0 0 0>, rotation=<rot -60 180 -180>, scale=<1 1 1>}}
    ]
    rotationLimits [{limitMax=<180 180 180>}, {limitMax=<180 180 180>}]
    rotationOffsets [<0 0 0>, <0 0 0>]
    trailBone {boneName=dyn_leaf_f_02}
   }
   n31 = Trail controller - Bone: dyn_leaf_g_02(ComponentPose=@n30, Alpha=@n20) {
    bShowLimit=false
    bUsePlanarLimit=true
    relaxationSpeedScale=5
    planarLimits [
     {drivingBone {boneName=dyn_leaf_g_01}, planeTransform {location=<0 0 0>, rotation=<rot 0 0 0>, scale=<1 1 1>}}
     {drivingBone {boneName=dyn_leaf_g_01}, planeTransform {location=<0 0 0>, rotation=<rot -60 180 -180>, scale=<1 1 1>}}
    ]
    rotationLimits [{limitMax=<180 180 180>}, {limitMax=<180 180 180>}]
    rotationOffsets [<0 0 0>, <0 0 0>]
    trailBone {boneName=dyn_leaf_g_02}
   }
   n32 = Trail controller - Bone: dyn_leaf_ha_02(ComponentPose=@n31, Alpha=@n20) {
    bShowLimit=false
    bUsePlanarLimit=true
    relaxationSpeedScale=6
    planarLimits [
     {drivingBone {boneName=dyn_leaf_ha_01}, planeTransform {location=<0 0 1>, rotation=<rot 0 0 0>, scale=<1 1 1>}}
     {drivingBone {boneName=dyn_leaf_ha_01}, planeTransform {location=<2 0 0>, rotation=<rot -60 180 -180>, scale=<1 1 1>}}
    ]
    rotationLimits [{limitMax=<180 180 180>}, {limitMax=<180 180 180>}]
    rotationOffsets [<0 0 0>, <0 0 0>]
    trailBone {boneName=dyn_leaf_ha_02}
   }
   n33 = Trail controller - Bone: dyn_leaf_hb_02(ComponentPose=@n32, Alpha=@n20) {
    bShowLimit=false
    bUsePlanarLimit=true
    relaxationSpeedScale=5
    planarLimits [
     {drivingBone {boneName=dyn_leaf_hb_01}, planeTransform {location=<0 0 1>, rotation=<rot 0 0 0>, scale=<1 1 1>}}
     {drivingBone {boneName=dyn_leaf_hb_01}, planeTransform {location=<2 0 0>, rotation=<rot -60 180 -180>, scale=<1 1 1>}}
    ]
    rotationLimits [{limitMax=<180 180 180>}, {limitMax=<180 180 180>}]
    rotationOffsets [<0 0 0>, <0 0 0>]
    trailBone {boneName=dyn_leaf_hb_02}
   }
   Save cached pose 'Cache_Trails'(Pose=Component To Local(ComponentPose=@n33)) {cacheName=Cache_Trails}
  ]}
 EventGraph {nodes [
   Event BlueprintUpdateAnimation()
   > Branch(Condition=OR Boolean(A=Get bRigidBodyEnabled(), B=Get bAnimDynamicsEnabled()))
   > n1 = Gravity Override State Function DEPRECATED(CharacterPartAnimInstance=Self-Reference(), GameTimeInSeconds=GetGameTimeInSeconds(), GravityOverrideParams=Get Gravity Override Params())
   > Set Gravity Override(Gravity Override=@n1)
  ]}
]
variables [
 AlphaTrail {type=number, default=0.5}
 Gravity Override {type=Vector}
 Gravity Override Params {type=GravityOverrideParamsStruct}
 SkeletalMesh {type=SkeletalMeshComponent}
]
